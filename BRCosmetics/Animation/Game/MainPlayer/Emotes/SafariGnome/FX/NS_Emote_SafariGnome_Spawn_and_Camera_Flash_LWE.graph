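# Niagara system for the SafariGnome emote's spawn and camera flash effect.
# Six emitters cover sparkles, airglow and smoke (LWE-suffixed, likely lightweight emitters) plus flash, flash star and backdrop.
# System scripts use variable assignments and SystemState, with front-end modifier, random range and bool-to-int dynamic inputs.
# User parameters expose front-end and preview flags and a normalized notify progress value.
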NiagaraSystem /BRCosmetics/Animation/Game/MainPlayer/Emotes/SafariGnome/FX/NS_Emote_SafariGnome_Spawn_and_Camera_Flash_LWE
dependencies {
 usedDataInterfaces []
 usedDynamicInputs [
  &FN_FrontEndVector3Modifier
  &Multiply_Vector2D_ByFloat
  &FN_FrontEndFloatModifier
  &RandomRangeFloat
  &MakeCustomIntFromBool
 ]
 usedModules [
  &self:SystemSpawnScript.SystemScriptSource.SystemScriptGraph.NiagaraNodeAssignment_0.SetVariables_#1
  &self:SystemSpawnScript.SystemScriptSource.SystemScriptGraph.NiagaraNodeAssignment_1.SetVariables_#2
  &self:SystemSpawnScript.SystemScriptSource.SystemScriptGraph.NiagaraNodeAssignment_2.SetVariables_#3
  &self:SystemSpawnScript.SystemScriptSource.SystemScriptGraph.NiagaraNodeAssignment_3.SetVariables_#4
  &SystemState
 ]
}
emitters {
 AirglowLWE {rendererClasses [], renderers []}
 Backdrop {rendererClasses [], renderers []}
 Flash {rendererClasses [], renderers []}
 FlashStar {rendererClasses [], renderers []}
 SmokeLWE {rendererClasses [], renderers []}
 SparklesLWE {rendererClasses [], renderers []}
}
summary {
 systemName=NS_Emote_SafariGnome_Spawn_and_Camera_Flash_LWE
 emitters [
  {emitterName=SparklesLWE, rendererClasses []}
  {emitterName=AirglowLWE, rendererClasses []}
  {emitterName=SmokeLWE, rendererClasses []}
  {emitterName=Flash, rendererClasses []}
  {emitterName=Backdrop, rendererClasses []}
  {emitterName=FlashStar, rendererClasses []}
 ]
 userVariables [
  User.bIsFrontEnd {type=NiagaraBool}
  User.bIsFrontEndPreview {type=NiagaraBool}
  User.NormalizedNotifyProgress {type=NiagaraFloat, defaultValue=0, description=""}
 ]
}
